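# Physics asset for the M_MED_Gator backpack cosmetic mesh, defining collision bodies and joint limits for its flaps, shells, teeth and pack bottom.
# Each bone gets a single capsule, with a tiny one on the pack bone and the largest on packBottom.
# All other bodies constrain directly to pack in a star layout rather than chains.
# Shells, teeth and flaps have tight swing and twist limits, while packBottom swings much more freely.
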PhysicsAsset /BRCosmetics/Accessories/FORT_Backpacks/Backpack_M_MED_Gator/Meshes/M_MED_Gator_Pack_Physics
bodies {
 flapFront [{shapeType=Capsule, length=4.1394, radius=2.1112, center=<1.7425 -0.3359 -0.1177>, rotation=<rot 90 0 0>}]
 flapSide [{shapeType=Capsule, length=2.3934, radius=1.8078, center=<1.8981 0.3338 -0.12>, rotation=<rot 0 0 -90>}]
 pack [{shapeType=Capsule, length=0, radius=0.5, center=<-0.5484 -7.9146 -2.9052>, rotation=<rot 90 0 0>}]
 packBottom [{shapeType=Capsule, length=14.9554, radius=8.716, center=<5.3891 0.5823 2.6999>, rotation=<rot 90 0 0>}]
 shell1 [{shapeType=Capsule, length=4.3106, radius=1.8607, center=<0.4965 0.0003 0>, rotation=<rot 90 0 0>}]
 shell2 [{shapeType=Capsule, length=4.3107, radius=1.8601, center=<0.5337 0 0.0001>, rotation=<rot 90 0 0>}]
 shell3 [{shapeType=Capsule, length=4.3106, radius=1.8606, center=<0.6052 0.0001 0.0001>, rotation=<rot 90 0 0>}]
 tooth1 [{shapeType=Capsule, length=1.5626, radius=1.2686, center=<1.3661 -0.1052 -0.1699>, rotation=<rot 90 0 0>}]
 tooth2 [{shapeType=Capsule, length=2.0296, radius=1.4058, center=<1.9086 0.0969 -0.1904>, rotation=<rot 90 0 0>}]
 tooth3 [{shapeType=Capsule, length=2.2675, radius=1.4664, center=<1.9662 -0.1092 -0.1548>, rotation=<rot 90 0 0>}]
 tooth4 [{shapeType=Capsule, length=2.7365, radius=1.8359, center=<2.8421 0.0168 -0.4058>, rotation=<rot 90 0 0>}]
}
constraints [
 {bone1Name=shell1, bone2Name=pack, swing1LimitDegrees=20, swing2LimitDegrees=5, twistLimitDegrees=2}
 {bone1Name=shell2, bone2Name=pack, swing1LimitDegrees=20, swing2LimitDegrees=5, twistLimitDegrees=2}
 {bone1Name=shell3, bone2Name=pack, swing1LimitDegrees=20, swing2LimitDegrees=5, twistLimitDegrees=2}
 {bone1Name=tooth1, bone2Name=pack, swing1LimitDegrees=15, swing2LimitDegrees=15, twistLimitDegrees=5}
 {bone1Name=flapSide, bone2Name=pack, swing1LimitDegrees=10, swing2LimitDegrees=15, twistLimitDegrees=2}
 {bone1Name=flapFront, bone2Name=pack, swing1LimitDegrees=10, swing2LimitDegrees=15, twistLimitDegrees=2}
 {bone1Name=tooth4, bone2Name=pack, swing1LimitDegrees=15, swing2LimitDegrees=15, twistLimitDegrees=5}
 {bone1Name=tooth3, bone2Name=pack, swing1LimitDegrees=15, swing2LimitDegrees=15, twistLimitDegrees=5}
 {bone1Name=tooth2, bone2Name=pack, swing1LimitDegrees=15, swing2LimitDegrees=15, twistLimitDegrees=5}
 {bone1Name=packBottom, bone2Name=pack, swing1LimitDegrees=45, swing2LimitDegrees=45, twistLimitDegrees=15}
]
